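# Temporary prototype material, named as a radius preview, for Niagara mesh particles in the TargetedBeam FX.
# It is translucent and unlit, with automatic usage flagging in the editor turned off.
# The graph holds only a Fresnel node and a Particle Color node, with no links shown between them or to the material outputs.
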Material /TargetedBeam/Assets/FX/Proto/M_TEMP_RadiusPreview
graph {
 properties {
  bAutomaticallySetUsageInEditor=false
  bUsedWithNiagaraMeshParticles=true
  blendMode=BLEND_Translucent
  shadingModel=MSM_Unlit
  usedShadingModels=Unlit
 }
 nodes [
  Fresnel()
  ParticleColor()
 ]
}
